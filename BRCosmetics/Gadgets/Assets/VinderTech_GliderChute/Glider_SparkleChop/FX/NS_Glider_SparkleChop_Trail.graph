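# Niagara trail effect for the SparkleChop glider, combining glinting sparkles with wind speedlines.
# Sparkles001 places flickering sprites at hand-authored array points, with time-of-day tint and camera offset, spawning faster as ForwardDot rises.
# Speedlines spawns short-lived, velocity-stretched streaks on a torus, skewed by RightDot and ForwardDot.
# Timeline time drives fade strength, while front-end switches retune rates, direction and opacity for preview.
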NiagaraSystem /BRCosmetics/Gadgets/Assets/VinderTech_GliderChute/Glider_SparkleChop/FX/NS_Glider_SparkleChop_Trail
dependencies {
 usedDataInterfaces [
  &NiagaraDataInterfaceCurve
  &NiagaraDataInterfaceVector2DCurve
  &NiagaraDataInterfaceColorCurve
  &NiagaraDataInterfaceCamera
  &NiagaraDataInterfaceArrayFloat3
  &NiagaraDataInterfaceSkeletalMesh
  &NiagaraDataInterfaceArrayFloat
 ]
 usedDynamicInputs [
  &SetBoolByFloatComparison
  &RemapRange
  &SelectFloatFromArray
  &MakeIntFromFloat
  &NormalizeVector
  &TransformVector
 ]
 usedModules [
  &self:SystemSpawnScript.SystemScriptSource.SystemScriptGraph.NiagaraNodeAssignment_0.SetVariables_#1
  &SystemState
  &Timeline
  &FN_Integrator
  &InterpolateOverTime
 ]
}
emitters {
 Sparkles001 {
  emitterUpdateScript [
   &EmitterState(
    Inactive Response=::NewEnumerator1
    Life Cycle Mode=::NewEnumerator1
    Loop Behavior=::NewEnumerator1
    Loop Duration Mode=::NewEnumerator1
    UseLoopDelay=false
   )
   &SpawnRate(
    Spawn Probability=0.5
    SpawnRate=&FN_FrontEndFloatModifier(Gameplay=&Lerp_Float(A=40, B=80, Alpha=$User.ForwardDot), Front End=40)
    Use Spawn Probability=true
   )
  ]
  particleSpawnScript [&InitializeParticle(
    Color=<rgba 3 2.841 2.9602 1>
    Lifetime Max=0.6
    Lifetime Min=0.3
    Lifetime Mode=::NewEnumerator1
    Mesh Scale Mode=::NewEnumerator3
    Position Mode=::NewEnumerator0
    Sprite Size Max=<33 55>
    Sprite Size Min=<17 25>
    Sprite Size Mode=::NewEnumerator2
    Write Material Random=true
   )]
  particleUpdateScript [
   &ParticleState()
   /SetVariables(Particles.Flicker:NiagaraFloat=&Waveform(
     Global Input Position=&RandomRangeFloat(Minimum=0, Maximum=9)
     "Waveform [1]"=::NewEnumerator10
     "[1] Frequency"=9
     "[1] Amplitude Min/Max"=<0.6 1>
    )) {isSetParameters=true}
   &ScaleSpriteSize(
    Uniform Curve Scale=$Particles.Flicker
    Uniform Curve Sprite Scale=DataInterface() {propertyValues {Curve {keys [
        {arriveTangent=9.0725, interpMode=RCIM_Cubic, leaveTangent=9.0725, tangentMode=RCTM_User, time=0, value=0}
        {arriveTangent=-2.5249, interpMode=RCIM_Cubic, leaveTangent=-2.5249, tangentMode=RCTM_User, time=0.1408, value=0.9847}
        {arriveTangent=-0.0405, interpMode=RCIM_Cubic, leaveTangent=-0.0405, tangentMode=RCTM_User, time=1, value=0}
       ]}}}
   )
   &ScaleColor(
    Scale Alpha=&FloatFromCurve(
     FloatCurve=DataInterface() {propertyValues {Curve {keys [
         {arriveTangent=9.0725, interpMode=RCIM_Cubic, leaveTangent=9.0725, tangentMode=RCTM_User, time=0, value=0}
         {arriveTangent=-2.5249, interpMode=RCIM_Cubic, leaveTangent=-2.5249, tangentMode=RCTM_User, time=0.1408, value=0.9847}
         {arriveTangent=-0.0405, interpMode=RCIM_Cubic, leaveTangent=-0.0405, tangentMode=RCTM_User, time=1, value=0}
        ]}}}
     Scale Curve=&Multiply_Float(
      A=$Particles.Flicker
      B=&FloatFromCurve(
       FloatCurve=DataInterface() {propertyValues {Curve {keys [
           {arriveTangentWeight=0.1667, interpMode=RCIM_Cubic, leaveTangentWeight=0.1667, tangentMode=RCTM_User, time=1.6, value=0}
           {tangentMode=RCTM_User, time=2, value=1}
          ]}}}
       CurveIndex=$Emitter.Age
       Scale Curve=&Multiply_Float(
        A=&Lerp_Float(A=0.7, B=0.8, Alpha=$System.Timeline.Time)
        B=&FN_FrontEndFloatModifier(Gameplay=1, Front End=0.75)
       )
      )
     )
    )
    Scale RGB=&MakeVectorFromLinearColorRGB(Color=&FNTimeOfDayColor(Night Color=<rgba 0.1257 0.4531 0.7396 1>, Morning Color=<rgba 1 1 1 1>, Day Color=<rgba 1 1 1 1>, Evening Color=<rgba 1 1 1 1>))
    ScaleA=true
    ScaleRGB=true
   )
   /SetVariables(Particles.BackSparkleColor:LinearColor=&MultiplyLinearColor(Linear Color=$Particles.Color, Scale Factor=&MakeVector4_FromFloat(Float=0.1))) {isSetParameters=true}
   /SetVariables(Particles.Distance:NiagaraFloat=&NormalizeDistanceRange(
     Start Position:NiagaraPosition=$Particles.Position
     End Position:NiagaraPosition=&ReturnCameraPosition(Coordinate Space:ENiagaraCoordinateSpace=::Simulation, Camera Query:NiagaraDataInterfaceCamera=DataInterface())
     Distance:NiagaraFloat=&FN_FrontEndFloatModifier(Gameplay=500, Front End=500)
     Invert Normalized Range:NiagaraBool=true
    )) {isSetParameters=true}
   &CameraOffset(Camera Offset Amount=&FN_FrontEndFloatModifier(Gameplay=&Lerp_Float(A=5, B=15, Alpha=$Particles.Distance), Front End=5)) {module=CameraOffset001}
   /SetVariables(Particles.Position=&SelectVectorFromArray(
     Array Sampling Mode=::NewEnumerator0
     Direct Array Index=?
     Vector Selection Array=DataInterface() {propertyValues {FloatData [
        <-57.8947 62.7833 8.7798>
        <58.5673 58.6829 15.3593>
        <58.6673 57.6 11.5493>
        <58.6673 55.9459 5.6942>
        <-33.8159 49.7952 5.8509>
        <-33.8159 53.499 6.1209>
        <-49.6194 62.8393 5.2557>
        <-41.1581 63.1101 12.873>
        <-38.4084 62.0353 3.3787>
        <-51.1806 57.8878 -3.3136>
        <-46.7271 60.8998 14.5407>
        <-35.9052 61.1105 14.3818>
        <-45.0829 54.4741 19.0922>
        <-38.2508 56.893 -3.7046>
        <-52.0286 49.1322 -3.7046>
        <-47.9544 48.876 19.0922>
        <-51.0939 48.6741 16.081>
        <-38.5048 48.6735 17.9787>
        <-43.4213 48.6778 3.134>
        <-35.1732 48.6794 -2.6685>
        <-44.9207 48.6776 3.7377>
        <-34.2483 48.9461 -0.3958>
        <33.8159 49.9289 13.7182>
        <33.8159 53.2727 9.7472>
        <33.8159 60.3662 4.862>
        <33.8159 58.8915 13.3808>
        <45.8539 62.958 8.5957>
        <42.1707 61.8837 3.134>
        <38.9686 60.2732 0.5354>
        <49.0005 63.0467 11.0898>
        <57.6151 55.404 18.6849>
        <38.9308 59.8765 15.3123>
        <42.7199 59.5999 15.5209>
        <52.8635 51.7235 19.0922>
        <34.9518 52.425 -3.7046>
        <41.9557 50.7791 -3.7046>
        <46.6135 51.3648 19.0922>
        <53.2885 48.6767 6.7551>
        <49.7662 48.6737 17.3161>
        <45.1553 48.6783 1.1406>
        <36.2325 48.6739 16.8256>
        <50.7961 48.6797 -3.6823>
        <38.3577 -73.193 10.2721>
        <17.567 -75.3136 10.1918>
        <30.3766 -81.0293 10.2987>
        <11.0969 -75.9443 10.189>
        <17.249 -78.0543 10.232>
        <-18.8341 -79.0059 10.25>
        <-25.0125 -83.9817 10.3401>
        <-15.5218 -85.2937 10.3487>
        <-29.358 -81.8816 10.3115>
        <-17.3759 -73.8505 10.1897>
        <-45.1452 -79.159 10.3077>
        <-49.1876 -76.4944 10.3045>
        <-66.4927 -70.9688 10.3637>
        <-55.3226 -66.2852 10.4005>
        <-63.2245 -61.4558 10.4775>
        <-47.6808 -58.4478 10.5673>
        <-55.4488 -46.6975 10.5353>
        <-56.7078 -40.3621 10.4693>
        <-47.1348 -47.957 10.6662>
        <-51.3969 -22.0837 10.4667>
        <-48.9824 1.3117 10.4144>
        <-49.4003 -12.124 10.4642>
        <-54.1559 1.443 10.3775>
        <-54.2093 1.2373 10.3776>
        <-51.9425 19.8798 10.3825>
        <-59.6666 19.7992 10.5067>
        <-56.6007 33.8451 10.5125>
        <-51.9613 23.8774 10.4151>
        <-47.816 33.6028 10.4057>
        <-49.2108 39.2645 10.468>
        <-63.0212 28.743 10.5596>
        <63.5578 -39.4641 10.534>
        <52.4155 -44.5769 10.5555>
        <54.9259 -54.1904 10.6036>
        <49.0981 -2.3736 10.4309>
        <48.2935 -5.5605 10.4612>
        <52.9076 -38.9863 10.5122>
        <46.4138 10.5687 10.4295>
        <55.3073 -2.6476 10.3814>
        <53.3396 14.9804 10.3558>
        <48.401 -67.1391 10.3996>
        <56.7469 -58.2941 10.544>
        <55.2438 18.4965 10.4356>
        <60.7467 18.139 10.5139>
        <50.6726 27.1366 10.4066>
        <52.7746 24.2467 10.4256>
        <52.9854 47.3535 10.5683>
        <49.8412 -71.016 10.3293>
        <59.1192 -68.9088 10.3743>
        <-58.3761 44.1817 15.1225>
        <-57.7126 46.6651 15.1225>
        <-58.2377 37.5559 15.121>
        <-62.1576 28.3186 15.1106>
        <-58.4492 17.2951 15.1082>
        <-53.0229 19.6233 15.1016>
        <-52.9454 6.9908 15.1271>
        <-52.0428 -0.5955 15.1033>
        <-53.1562 -48.3149 15.0628>
        <-58.8741 -46.9275 15.1267>
        <-62.2085 -44.2629 15.1758>
        <-53.2944 -64.5047 15.0588>
        <-52.8249 -68.9693 15.0774>
        <-61.3567 -71.1776 15.1734>
        <-8.521 -72.8398 15.0996>
        <-26.6772 -83.0644 15.1835>
        <-26.8341 -78.6174 15.1723>
        <-6.6209 -82.7816 15.2407>
        <-18.8741 -84.0392 15.2532>
        <-16.8257 -75.3314 15.1297>
        <2.7215 -80.1999 15.2056>
        <43.315 -80.2239 15.188>
        <2.235 -82.7816 15.2427>
        <2.7808 -77.5544 15.1677>
        <41.74 -76.073 15.1363>
        <56.4563 38.9345 15.1225>
        <54.9537 23.0797 15.1162>
        <57.8024 22.8579 15.1127>
        <56.7831 12.7077 15.103>
        <55.7677 -3.1105 15.1348>
        <52.8516 13.9633 15.1143>
        <55.2612 -35.059 15.1987>
        <51.8652 -42.392 15.046>
        <52.6509 -11.5265 15.1121>
        <56.7625 -37.3793 15.1795>
        <57.7735 -48.3298 15.1074>
        <48.6543 -61.0536 15.0424>
        <52.35 -56.0072 15.0235>
        <52.3118 -64.698 15.0597>
        <52.8472 -69.3569 15.0861>
        <56.5031 -73.7364 15.1317>
        <65.0088 -74.3738 15.1047>
        <59.8689 -77.3201 15.1475>
        <-60.4991 48.6539 11.1522>
        <-59.4118 48.3158 15.0353>
        <-64.2901 22.4758 12.1403>
        <-56.3359 4.7125 11.2223>
        <-56.7215 -19.641 12.9759>
        <49.5707 48.9686 12.6718>
        <64.0301 46.9084 13.7818>
        <56.286 -26.2571 14.8624>
        <56.9303 -29.4748 12.8403>
        <44.7185 -81.6148 12.0191>
        <26.8711 -84.5718 12.2408>
        <44.1128 -81.542 13.1251>
        <19.3275 -85.0608 14.2577>
        <-32.5665 -83.5447 11.9973>
        <-14.6902 -86.4195 12.5591>
        <-32.6007 -83.1079 10.9947>
        <-67.9608 -72.9192 14.4086>
        <-63.8003 -46.2575 14.1953>
        <68.6318 -64.1638 12.7152>
        <64.1203 -41.1903 13.6762>
        <58.657 -33.8147 11.6066>
       ]}}
     Random Seed=0
    )) {isSetParameters=true}
  ]
  rendererClasses [&NiagaraSpriteRendererProperties]
  renderers [{rendererClass=&NiagaraSpriteRendererProperties}]
 }
 Speedlines {
  emitterUpdateScript [
   &EmitterState()
   &SpawnRate(
    SpawnRate=&MakeCustomFloatFromBool(Bool=$User.bIsFrontEndPreview, True Float=50, False Float=90)
    Use Spawn Probability=true
   ) {module=SpawnRate001}
  ]
  particleSpawnScript [
   &InitializeParticle(
    Color=&MakeLinearColorFromVectorAndFloat(
     "Vector (RGB)"=<1 1 1>
     "Float (Alpha)"=&Multiply_Float(
      A=&UniformRangedFloat(Minimum=1, Maximum=1)
      B=&MakeCustomFloatFromBool(
       Bool=$User.bIsFrontEndPreview
       True Float=&FloatFromCurve(
        FloatCurve=DataInterface() {propertyValues {Curve {keys [{tangentMode=RCTM_User, time=0, value=0}, {tangentMode=RCTM_User, time=2, value=1}]}}}
        CurveIndex=$Emitter.Age
        Scale Curve=0.4
       )
       False Float=1
      )
     )
    )
    Lifetime Max=0.2
    Lifetime Min=0.15
    Lifetime Mode=::NewEnumerator1
    Position Offset=&FN_FrontEndVector3Modifier(Gameplay=<0 -50 0>, Front End=<0 50 0>)
    Sprite Size Max=<3 10>
    Sprite Size Min=<2 2>
    Sprite Size Mode=::NewEnumerator2
    UsePositionOffset=true
    Write Material Random=true
   )
   &AddVelocity(
    Module.Use Falloff Distance=false
    Override Seed=false
    Velocity Mode=::NewEnumerator1
    Velocity Speed=&FN_FrontEndFloatModifier(Gameplay=-2500, Front End=2500)
   ) {module=AddVelocity001}
   &TorusLocation(
    Align Torus With Axis=true
    Coordinate Space:ENiagaraCoordinateSpace=::Simulation
    Handle Radius=0
    Large Radius:NiagaraFloat=&FN_FrontEndFloatModifier(
     Gameplay=&RandomRangeFloat(Minimum=450, Maximum=500)
     Front End=&RandomRangeFloat(Minimum=250, Maximum=300)
    )
    MaskSpawnGroup=false
    Module.Override Randomness=false
    Offset=<0 33 0>
    Torus Axis=<0 1 0>
    Torus Origin:NiagaraPosition=&AddVectorToPosition(
     Position=$Particles.Position
     Vector=&FN_FrontEndVector3Modifier(Gameplay=<0 -500 0>, Front End=<0 -500 0>)
    )
   )
   &AddVelocity(
    Velocity=&Add_Vector(
     A=&RandomRangeVector(Minimum=<0 300 0>, Maximum=<0 700 1>)
     B=&Add_Vector(
      A=&Multiply_VectorByFloat(Vector=<-555 0 0>, Float=$User.RightDot)
      B=&Multiply_VectorByFloat(Vector=<0 0 222>, Float=$User.ForwardDot)
     )
    )
    Velocity Speed Scale=&FN_FrontEndFloatModifier(Gameplay=1, Front End=0.3)
   )
  ]
  particleUpdateScript [
   &SolveForcesAndVelocity(Clamp Velocity=false, Limit Acceleration=false)
   &ParticleState()
   &ScaleSpriteSize(Scale Factor=&Vector2DFromCurve(
     Vector2Curve=DataInterface() {propertyValues {XCurve {keys [<at 0 0.5 RCIM_Cubic>,<at 0.5 1 RCIM_Cubic>]}, YCurve {keys [<at 0 0 RCIM_Cubic>,<at 0.5 1 RCIM_Cubic>]}}}
     ScaleCurve=&Vector2DFromFloat(Value=1)
    ))
   &ScaleSpriteSizeBySpeed(Max Scale Factor=<1 50>, Min Scale Factor=<0.01 0.05>, Velocity Threshold=1)
   &ScaleColor(
    Scale Alpha=&FloatFromCurve(
     FloatCurve=DataInterface() {propertyValues {Curve {keys [
         <at 0 0 RCIM_Cubic>
         <at 0.3759 1.0207>
         {arriveTangent=-2.221, interpMode=RCIM_Cubic, leaveTangent=-2.221, tangentMode=RCTM_User, time=1, value=0}
        ]}}}
     Scale Curve=&MakeCustomFloatFromBool(
      Bool=$User.bIsFrontEndPreview
      True Float=&MakeCustomFloatFromBool(Bool=$User.bIsFullyDeployed, True Float=0.5, False Float=0)
      False Float=&Lerp_Float(A=0, B=0.5, Alpha=&FN_FrontEndFloatModifier(Gameplay=$System.Timeline.Time, Front End=0.5))
     )
    )
    Scale RGB=&VectorFromFloat(Value=&FN_FrontEndFloatModifier(Gameplay=0.5, Front End=0.33))
    ScaleA=true
    ScaleRGB=true
   )
  ]
  rendererClasses [&NiagaraSpriteRendererProperties]
  renderers [{rendererClass=&NiagaraSpriteRendererProperties}]
 }
}
summary {
 systemName=NS_Glider_SparkleChop_Trail
 emitters [
  {emitterName=Speedlines, rendererClasses [&NiagaraSpriteRendererProperties]}
  {emitterName=Sparkles001, rendererClasses [&NiagaraSpriteRendererProperties]}
 ]
 userVariables [
  User.bIsFrontEnd {type=NiagaraBool}
  User.bIsFrontEndPreview {type=NiagaraBool}
  User.bIsFullyDeployed {type=NiagaraBool}
  User.ForwardDot {type=NiagaraFloat, defaultValue=0, description=""}
  User.RibbonIDOffset {type=NiagaraInt32}
  User.RightDot {type=NiagaraFloat, defaultValue=0, description=""}
  User.Variant {type=NiagaraFloat, defaultValue=0, description=""}
 ]
}
